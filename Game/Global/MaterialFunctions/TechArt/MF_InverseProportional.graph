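# Math material function, exposed to the library, that creates an inverse proportional falloff with adjustable steepness.
# Takes k as the reciprocal of Tightness and outputs k/x - k, which crosses zero at x = 1 and levels off at -k for large inputs.
# Both inputs are floored at zero, so the result blows up as x nears zero; the main input previews with the U texture coordinate.
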MaterialFunction /Game/Global/MaterialFunctions/TechArt/MF_InverseProportional
graph {
 properties {
  bExposeToLibrary=true
  description="Creates an inverse proportional falloff with adjustable steepness.Input values approaching 0 from the positive direction will exceed the 0-1 range and become (clamped) infinite at 0.High positive input values will approach 0."
  libraryCategoriesText [Math]
 }
 nodes [
  n1 = FunctionInput() {bUsePreviewValueAsDefault=true, inputName=Tightness, inputType=FunctionInput_Scalar, previewValue=<100 0 0 1>}
  n2 = Divide(B=Max(A=@n1) {constB=0})
  n3 = TextureCoordinate()
  n4 = Max(A=FunctionInput(Preview=ComponentMask(None=@n3) {r=true}) {inputType=FunctionInput_Scalar}) {constB=0}
  FunctionOutput(None=Subtract(A=Divide(A=@n2, B=@n4), B=@n2))
 ]
}
