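# Parent translucent unlit material for the Rainbow Raid lightning gadget cosmetic, drawing animated lightning over the character PreFX/PostFX base functions.
# Noise-offset, quantized time cycles and step masks form big and small strands, flipbooking bolts modeled in the mesh.
# SplineThicken supplies bolt width and parallax UVs, while world position offset adds sine jitter and can collapse geometry to the object origin.
# Panned Voronoi, gradient masks and three tint colors build the emissive, with depth-faded opacity.
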Material /BRCosmetics/Animation/Props/Gadgets_RainbowRaid_lightning/Material/M_RainbowRaid_LightingGadget_Parent
graph {
 properties {
  bAutomaticallySetUsageInEditor=false
  bHasPixelAnimation=true
  bUseMaterialAttributes=true
  bUsedWithMeshDeformer=true
  bUsedWithMorphTargets=true
  bUsedWithSkeletalMesh=true
  blendMode=BLEND_Translucent
  pixelDepthOffsetMode=PDOM_Legacy
  shadingModel=MSM_Unlit
  twoSided=true
  usedShadingModels=Unlit
 }
 nodes [
  n1 = MaterialFunctionCall() {materialFunction=&MF_CharacterParent_2023_PreFX}
  n2 = NamedRerouteUsage()
  MaterialFunctionCall(
   Material=SetMaterialAttributes!(
    MaterialAttributes=GetMaterialAttributes(None=@n1.Result).MaterialAttributes
    Emissive Color=NamedRerouteUsage()
    World Position Offset=NamedRerouteUsage()
    Opacity=@n2
   ) {attributeSetTypes [#1, #2, #3]}
   M=@n1.M
   Pre Skinned Local Position=@n1.M
   Distance Blend=@n1.M
   Fresnel=@n1.M
  ) {materialFunction=&MF_CharacterParent_2023_PostFX}
  n3 = Multiply(A=NamedRerouteUsage()) {constB=131.1137}
  Multiply(A=LinearInterpolate(Alpha=Frac(None=@n3)) {constA=0.03}, B=@n2)
  n4 = Multiply(A=Time(), B=ScalarParameter() {defaultValue=0.15, group=Thunder, parameterName=MeshFlipbookSpeed, sortPriority=1})
  n5 = TextureSample(UVs=@n4) {samplerType=SAMPLERTYPE_Masks, texture=/Game/Global/Textures/Default/Noise/T_Generic_Perlin}
  n6 = Multiply(A=Time(), B=ScalarParameter() {defaultValue=0.25, group=Thunder, parameterName=MeshFlipboookCyclingSpeed, sortPriority=1}) {constB=0.25}
  n7 = Constant() {r=10}
  n8 = Floor(None=Multiply(A=Add(A=@n5.R, B=@n6), B=@n7))
  NamedRerouteDeclaration(None=Frac(None=Divide(A=@n8, B=@n7))) {name=Time0-1Big, nodeColor=<rgba 1 0 0 0>}
  n9 = ObjectPositionWS()
  n10 = WorldPosition() {worldPositionShaderOffset=WPT_CameraRelativeNoOffsets}
  n11 = NamedRerouteUsage()
  n12 = Constant() {r=43758.5469}
  n13 = NamedRerouteUsage()
  n14 = Time()
  n15 = Multiply(A=NamedRerouteUsage()) {constB=51.1437}
  n16 = Add(A=Frac(None=Add(A=@n12, B=@n13)), B=Add(A=Multiply(A=@n14) {constB=0.8}, B=Frac(None=@n15)))
  n17 = Multiply(A=VertexNormalWS(), B=Constant3Vector() {constant=<rgba 4 4 8 0>})
  n18 = Multiply(A=@n17) {constB=-1}
  n19 = Add(
   A=@n11
   B=Multiply(
    A=MaterialFunctionCall!(Sine Phase=@n16, Value 1=@n18, Value 2=@n17) {materialFunction=/Engine/Functions/Engine_MaterialFunctions02/Utility/Sine_Remapped}.Remapped Sine
    B=Constant3Vector() {constant=<rgba 1 1 1 0>}
   )
  )
  n20 = NamedRerouteUsage()
  NamedRerouteDeclaration(None=LinearInterpolate(A=Subtract(A=@n9, B=@n10.XYZ), B=LinearInterpolate(A=@n11, B=@n19, Alpha=@n20), Alpha=NamedRerouteUsage())) {name=FinalWPO, nodeColor=<rgba 0.5 1 0 1>}
  n21 = TextureSample(UVs=Add(A=@n4) {constB=0.4154}) {samplerType=SAMPLERTYPE_Masks, texture=&T_Generic_Perlin}
  n22 = Floor(None=Multiply(A=Add(A=@n21.R, B=@n6), B=@n7))
  NamedRerouteDeclaration(None=Frac(None=Divide(A=@n22, B=@n7))) {name=Time0-1Small, nodeColor=<rgba 1 0 0.4219 1>}
  n23 = NamedRerouteUsage()
  n24 = ScalarParameter() {defaultValue=3, group=Thunder, parameterName=ParallaxUmaskPow, sortPriority=4}
  n25 = Saturate(None=Multiply(A=Power(Base=@n23, Exp=@n24), B=ScalarParameter() {defaultValue=55, group=Thunder, parameterName=ParallaxUMaskMult, sortPriority=4}) {constB=2})
  n26 = Multiply(A=Multiply(A=NamedRerouteUsage(), B=@n25), B=NamedRerouteUsage())
  n27 = LinearInterpolate(A=NamedRerouteUsage(), B=NamedRerouteUsage(), Alpha=NamedRerouteUsage())
  n28 = LinearInterpolate(A=NamedRerouteUsage(), B=@n27, Alpha=@n25)
  n29 = NamedRerouteUsage()
  n30 = NamedRerouteUsage()
  n31 = ScalarParameter() {defaultValue=1.5, group=Thunder, parameterName=ParallaxVmaskPow, sortPriority=5}
  n32 = Saturate(None=Multiply(A=Power(Base=@n30, Exp=@n31), B=ScalarParameter() {defaultValue=8, group=Thunder, parameterName=ParallaxVMaskMult, sortPriority=5}) {constB=2})
  NamedRerouteDeclaration(None=Multiply(A=Add(A=@n26, B=@n28), B=LinearInterpolate(A=@n29, B=@n27, Alpha=@n32))) {name=FinalColor, nodeColor=<rgba 1 0 0.8203 1>}
  n33 = Multiply(
   A=Power(Base=NamedRerouteUsage(), Exp=ScalarParameter() {defaultValue=2, group=Thunder, parameterName=ThunderOpacityPow, sortPriority=3})
   B=ScalarParameter() {defaultValue=8, group=Thunder, parameterName=ThunderOpacityMult, sortPriority=3}
  ) {constB=2}
  NamedRerouteDeclaration(None=Saturate(None=DepthFade(Opacity=@n33) {fadeDistanceDefault=30})) {name=FinalTranslucent, nodeColor=<rgba 1 0 0.2109 1>}
  NamedRerouteDeclaration(None=VectorParameter() {group=Thunder, parameterName=OuterColor, sortPriority=7, defaultValue=<rgba 0 0.0822 1 1>}.RGB) {name=OuterColor, nodeColor=<rgba 0 1 0.8359 1>}
  NamedRerouteDeclaration(None=VectorParameter() {group=Thunder, parameterName=InnerColor, sortPriority=7, defaultValue=<rgba 0 1.25 1.25 1>}.RGB) {name=InnerColor, nodeColor=<rgba 1 0.5156 0 1>}
  NamedRerouteDeclaration(None=VectorParameter() {group=Thunder, parameterName=BrightColor, sortPriority=7, defaultValue=<rgba 0 2 1.2174 1>}.RGB) {name=BrightColor, nodeColor=<rgba 0.1953 0 1 1>}
  n34 = NamedRerouteUsage()
  n35 = ComponentMask(None=@n34) {r=true}
  n36 = OneMinus(None=@n35)
  NamedRerouteDeclaration(None=Multiply(A=Multiply(A=@n35, B=@n36)) {constB=4}) {name=VGradient, nodeColor=<rgba 0 0.7187 1 1>}
  n37 = TextureCoordinate()
  n38 = LinearInterpolate(A=Constant2Vector() {g=0.5, r=3}, B=Constant2Vector() {g=0.25, r=0.5}, Alpha=NamedRerouteUsage())
  n39 = Time()
  n40 = NamedRerouteUsage()
  n41 = TextureSample(UVs=Panner(Coordinate=Multiply(A=@n37, B=@n38), Time=Add(A=@n39, B=@n40)) {speedX=2, speedY=0.5}) {texture=/BRCosmetics/Weapons/FORT_Melee/Pickaxe_BiomeHose/FX/T_Pickaxe_BiomeHose_Voronoi_Inverted}
  NamedRerouteDeclaration(None=Multiply(A=SmoothStep(Value=@n41.R) {constMax=0.5}) {constB=2}) {name=ThunderTexture, nodeColor=<rgba 0 1 0.1562 1>}
  n42 = Abs(None=Subtract(A=ComponentMask(None=@n34) {g=true}) {constB=0.5})
  n43 = Multiply(A=@n42) {constB=2}
  n44 = OneMinus(None=@n43)
  n45 = LinearInterpolate(
   A=ScalarParameter() {defaultValue=100, group=Thunder, parameterName=Thunder TightnessBig}
   B=ScalarParameter() {defaultValue=30, group=Thunder, parameterName=Thunder TightnessSmall}
   Alpha=NamedRerouteUsage()
  )
  NamedRerouteDeclaration(None=Multiply(
    A=Power(Base=@n44) {constExponent=0.5}
    B=MaterialFunctionCall!(In=@n43, Tightness=@n45) {materialFunction=/Game/Global/MaterialFunctions/TechArt/MF_InverseProportional}.Result
   )) {name=Opacity_UGradient, nodeColor=<rgba 0.5 0 1 1>}
  NamedRerouteDeclaration(None=Add(A=VertexColor().R) {constB=-0.01}) {name=VertexRMask, nodeColor=<rgba 1 0 0 0>}
  n46 = ComponentMask(None=TextureCoordinate()) {r=true}
  n47 = OneMinus(None=@n46)
  n48 = Multiply(
   A=LinearInterpolate(
    A=ScalarParameter() {defaultValue=3, group=Thunder, parameterName=SmallThunderSize, sortPriority=2}
    B=ScalarParameter() {defaultValue=12, group=Thunder, parameterName=BigThunderSize, sortPriority=2}
    Alpha=NamedRerouteUsage()
   ) {constA=2, constB=10}
   B=Multiply(A=@n46, B=SmoothStep(Value=@n47) {constMax=0.2})
  )
  n49 = MaterialFunctionCall(WidthBase=@n48, WidthTip=@n48, Expand U or V UV Channel=StaticBoolParameter() {parameterName="Expand U or V UV Channel (SB)"}) {materialFunction=&SplineThicken}
  NamedRerouteDeclaration(None=@n49.WorldPositionOffset) {name=SplineWPO, nodeColor=<rgba 0.3828 1 0 1>}
  NamedRerouteDeclaration(None=@n49.Normal) {name=Normal, nodeColor=<rgba 1 0 0.7031 1>}
  NamedRerouteDeclaration(None=@n49.UVs with Parallax) {name=ParallaxUVs, nodeColor=<rgba 0 0.625 1 1>}
  n50 = NamedRerouteUsage()
  n51 = NamedRerouteUsage()
  n52 = NamedRerouteUsage()
  n53 = Subtract(A=NamedRerouteUsage()) {constB=0.1}
  n54 = NamedRerouteDeclaration(None=Subtract(A=Step(Y=@n50, X=@n51), B=Step(Y=@n52, X=@n53))) {name=BigStrand, nodeColor=<rgba 1 0.4219 0 1>}
  n55 = NamedRerouteUsage()
  n56 = NamedRerouteUsage()
  n57 = NamedRerouteUsage()
  n58 = Subtract(A=NamedRerouteUsage()) {constB=0.1}
  n59 = NamedRerouteDeclaration(None=Subtract(A=Step(Y=@n55, X=@n56), B=Step(Y=@n57, X=@n58))) {name=SmallStrand, nodeColor=<rgba 0.7109 0 1 1>}
  NamedRerouteDeclaration(None=Saturate(None=Add(A=@n54, B=@n59))) {name=FinalOpacity, nodeColor=<rgba 0.2656 1 0 1>}
  n60 = ScalarParameter() {defaultValue=1, group=Thunder, parameterName=Thunder Emissive InvScale}
  NamedRerouteDeclaration(None=MaterialFunctionCall(In=Multiply(A=@n42, B=@n60) {constB=2}, Tightness=@n45) {materialFunction=&MF_InverseProportional}.Result) {name=Emissive_UGradient, nodeColor=<rgba 1 0 0.3984 1>}
  n61 = ComponentMask(None=TextureCoordinate()) {r=true}
  n62 = Multiply(A=Multiply(A=@n61, B=OneMinus(None=@n61))) {constB=4}
  NamedRerouteDeclaration(None=Saturate(None=Power(Base=@n62) {constExponent=4})) {name=CenterMask, nodeColor=<rgba 0.4531 0 1 1>}
  TemporalResponsivenessOutput!(None=Constant() {r=0.4})
 ]
}
